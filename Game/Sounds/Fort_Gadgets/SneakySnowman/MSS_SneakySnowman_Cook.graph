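# One-shot mono MetaSound Source for the Sneaky Snowman gadget's cook sound.
# On play, a random get picks one of two cook wave assets and triggers a wave player, whose end fires OnFinished.
# A random float scales the player's output for slight volume variation.
# It reuses the default grenade cook attenuation and routes to the gadget items sound class.
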MetaSoundSource /Game/Sounds/Fort_Gadgets/SneakySnowman/MSS_SneakySnowman_Cook
inputs [
 In Array {type=WaveAsset:Array}
 UE.Source.OnPlay {type=Trigger}
]
interfaces [UE.OutputFormat.Mono:1.0, UE.Source.OneShot:1.0, UE.Source:1.1]
outputs [
 "UE.OutputFormat.Mono.Audio:0" {type=Audio}
 UE.Source.OneShot.OnFinished {type=Trigger}
]
pages [Default {
  input_defaults {
   UE.Source.OnPlay=null
   In Array [
    /Game/Sounds/Fort_Gadgets/SneakySnowman/CookThrow/SneakySnowman_Cook_01
    /Game/Sounds/Fort_Gadgets/SneakySnowman/CookThrow/SneakySnowman_Cook_02
   ]
  }
  nodes [
   ArrayRandomGet_0 = Array.Random Get:WaveAsset:Array(In Array=^In Array, Next=^UE.Source.OnPlay)
   WavePlayer_0 = UE.Wave Player:Mono(Play=@ArrayRandomGet_0.On Next, Wave Asset=@ArrayRandomGet_0.Value) {also {version [1,1]}}
   Multiply_0 = UE.Multiply:Audio by Float(AdditionalOperands=UE.RandomFloat() {also {defaults {Max=1.2, Min=0.8}}}.Value, PrimaryOperand=@WavePlayer_0.Out Mono)
  ]
  links [{dst="@UE.OutputFormat.Mono.Audio:0", src=Multiply_0.Out}, {dst="@UE.Source.OneShot.OnFinished", src=WavePlayer_0.On Finished}]
 }]
properties {
 assetImportData=&self:AssetImportData
 attenuationSettings=/Game/Sounds/Fort_Weapons/Grenades/Grenade_Default_Cook_Att
 maxDistance=3000
 sampleRate=48000
 soundClassObject=&Fort_Gadgets_Items
 volume=0.5
}
