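# One-shot stereo MetaSound for the Sneaky Snowman gadget's throw, using player foley class and attenuation.
# On play it layers a start wave, a random pitch-shifted throw and, at 50% probability, a random sweetener, mixed with randomized gains.
# OnFinished waits only for the start and throw layers, not the sweetener.
# A Mobile page drops the sweetener for a two-input mix with its own gain ranges.
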MetaSoundSource /Game/Sounds/Fort_Gadgets/SneakySnowman/MSS_SneakySnowman_Throw
inputs [
 Grenade Start {type=WaveAsset:Array}
 Grenade Throw {type=WaveAsset:Array}
 Grenade Throw Sweetener {type=WaveAsset:Array}
 UE.Source.OnPlay {type=Trigger}
]
interfaces [UE.OutputFormat.Stereo:1.0, UE.Source.OneShot:1.0, UE.Source:1.1]
outputs [
 "UE.OutputFormat.Stereo.Audio:0" {type=Audio}
 "UE.OutputFormat.Stereo.Audio:1" {type=Audio}
 UE.Source.OneShot.OnFinished {type=Trigger}
]
pages [
 Default {
  is_build_graph=true
  input_defaults {
   UE.Source.OnPlay=null
   Grenade Start [/Game/Sounds/Fort_Gadgets/SneakySnowman/CookThrow/Molotov_Throw]
   Grenade Throw [
    /Game/Sounds/Fort_Gadgets/SneakySnowman/CookThrow/SneakySnowman_Throw_03
    /Game/Sounds/Fort_Gadgets/SneakySnowman/CookThrow/SneakySnowman_Throw_02
    /Game/Sounds/Fort_Gadgets/SneakySnowman/CookThrow/SneakySnowman_Throw_01
   ]
   Grenade Throw Sweetener [
    /Game/Sounds/Fort_Gadgets/SneakySnowman/CookThrow/ChillGrenade_Throw_03
    /Game/Sounds/Fort_Gadgets/SneakySnowman/CookThrow/ChillGrenade_Throw_02
    /Game/Sounds/Fort_Gadgets/SneakySnowman/CookThrow/ChillGrenade_Throw_01
   ]
  }
  nodes [
   TriggerFilter_0 = UE.Trigger Filter(Trigger=^UE.Source.OnPlay) {also {defaults {Probability=0.5}}}
   ArrayRandomGet_1 = Array.Random Get:WaveAsset:Array(In Array=^Grenade Throw Sweetener, Next=@TriggerFilter_0.Heads) {also {defaults {Enable Shared State="<MetaSoundBuilderResult.SUCCEEDED: 0>", Weights []}}}
   WavePlayer_0 = UE.Wave Player:Stereo(Play=@ArrayRandomGet_1.On Next, Wave Asset=@ArrayRandomGet_1.Value) {also {version [1,1]}}
   ArrayRandomGet_0 = Array.Random Get:WaveAsset:Array(In Array=^Grenade Throw, Next=^UE.Source.OnPlay) {also {defaults {Enable Shared State="<MetaSoundBuilderResult.SUCCEEDED: 0>"}}}
   WavePlayer_1 = UE.Wave Player:Stereo(Play=@ArrayRandomGet_0.On Next, Wave Asset=@ArrayRandomGet_0.Value, Pitch Shift=UE.RandomFloat() {also {defaults {Max=2, Min=-2}}}.Value) {also {version [1,1]}}
   ArrayRandomGet_2 = Array.Random Get:WaveAsset:Array(In Array=^Grenade Start, Next=^UE.Source.OnPlay) {also {defaults {Enable Shared State="<MetaSoundBuilderResult.SUCCEEDED: 0>", Weights []}}}
   WavePlayer_2 = UE.Wave Player:Stereo(Play=@ArrayRandomGet_2.On Next, Wave Asset=@ArrayRandomGet_2.Value) {also {version [1,1]}}
   AudioMixerAudioMixerStereo3_0 = "AudioMixer.Audio Mixer (Stereo, 3)"(
    Gain 1=UE.RandomFloat(Next=^UE.Source.OnPlay) {also {defaults {Max=0.8, Min=0.5}}}.Value
    Gain 0=UE.RandomFloat(Next=^UE.Source.OnPlay) {also {defaults {Max=0.6, Min=0.5}}}.Value
    Gain 2=UE.RandomFloat(Next=^UE.Source.OnPlay) {also {defaults {Min=0.6}}}.Value
    In 2 L=@WavePlayer_0.Out Left
    In 2 R=@WavePlayer_0.Out Right
    In 0 L=@WavePlayer_1.Out Left
    In 0 R=@WavePlayer_1.Out Right
    In 1 L=@WavePlayer_2.Out Left
    In 1 R=@WavePlayer_2.Out Right
   )
   TriggerAccumulatorTriggerAccumulate2_0 = "TriggerAccumulator.Trigger Accumulate (2)"(In 0=@WavePlayer_1.On Finished, In 1=@WavePlayer_2.On Finished)
  ]
  links [
   {dst="@UE.OutputFormat.Stereo.Audio:0", src=AudioMixerAudioMixerStereo3_0.Out L}
   {dst="@UE.OutputFormat.Stereo.Audio:1", src=AudioMixerAudioMixerStereo3_0.Out R}
   {dst="@UE.Source.OneShot.OnFinished", src=TriggerAccumulatorTriggerAccumulate2_0.Out}
  ]
 }
 Mobile {
  nodes [
   ArrayRandomGet_0 = Array.Random Get:WaveAsset:Array(In Array=^Grenade Throw, Next=^UE.Source.OnPlay) {also {defaults {Enable Shared State="<MetaSoundBuilderResult.SUCCEEDED: 0>"}}}
   WavePlayer_0 = UE.Wave Player:Stereo(Play=@ArrayRandomGet_0.On Next, Wave Asset=@ArrayRandomGet_0.Value, Pitch Shift=UE.RandomFloat() {also {defaults {Max=2, Min=-2}}}.Value) {also {version [1,1]}}
   ArrayRandomGet_1 = Array.Random Get:WaveAsset:Array(In Array=^Grenade Start, Next=^UE.Source.OnPlay) {also {defaults {Enable Shared State="<MetaSoundBuilderResult.SUCCEEDED: 0>", Weights []}}}
   WavePlayer_1 = UE.Wave Player:Stereo(Play=@ArrayRandomGet_1.On Next, Wave Asset=@ArrayRandomGet_1.Value) {also {version [1,1]}}
   AudioMixerAudioMixerStereo2_0 = "AudioMixer.Audio Mixer (Stereo, 2)"(
    Gain 1=UE.RandomFloat(Next=^UE.Source.OnPlay) {also {defaults {Max=1.3, Min=0.7}}}.Value
    Gain 0=UE.RandomFloat(Next=^UE.Source.OnPlay) {also {defaults {Min=0.5}}}.Value
    In 0 L=@WavePlayer_0.Out Left
    In 0 R=@WavePlayer_0.Out Right
    In 1 L=@WavePlayer_1.Out Left
    In 1 R=@WavePlayer_1.Out Right
   )
   TriggerAccumulatorTriggerAccumulate2_0 = "TriggerAccumulator.Trigger Accumulate (2)"(In 0=@WavePlayer_0.On Finished, In 1=@WavePlayer_1.On Finished)
  ]
  links [
   {dst="@UE.OutputFormat.Stereo.Audio:0", src=AudioMixerAudioMixerStereo2_0.Out L}
   {dst="@UE.OutputFormat.Stereo.Audio:1", src=AudioMixerAudioMixerStereo2_0.Out R}
   {dst="@UE.Source.OneShot.OnFinished", src=TriggerAccumulatorTriggerAccumulate2_0.Out}
  ]
 }
]
properties {
 assetImportData=&self:AssetImportData
 attenuationSettings=&Foley_Attenuation_Basic_Att
 maxDistance=1536
 numChannels=2
 outputFormat=Stereo
 sampleRate=48000
 soundClassObject=&Fort_Player_Foley
 volume=2
}
