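# Sound cue for the Sneaky Snowman gadget's ball destroy sound.
# A single wave player feeds the SneakySnowman_Destroy_01 wave straight to the output, with a boosted volume and no randomization.
# It uses the gadget items sound class and borrows the grenade bounce attenuation.
# A concurrency override allows one instance per owner and blocks new plays while it sounds.
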SoundCue /Game/Sounds/Fort_Gadgets/SneakySnowman/SneakySnowman_BallDestroy_Cue
root=WavePlayer_1
nodes [{id=WavePlayer_1, properties {soundWaveAssetPtr=/Game/Sounds/Fort_Gadgets/SneakySnowman/Version2/SneakySnowman_Destroy_01}}]
properties {
 attenuationSettings=/Game/Sounds/Attenuation/Fort_Grenade_Bounce_Att
 bOverrideConcurrency=true
 maxDistance=5500
 soundClassObject=&Fort_Gadgets_Items
 volumeMultiplier=1.3333
 concurrencyOverrides {bLimitToOwner=true, maxCount=1, resolutionRule=PreventNew}
}
