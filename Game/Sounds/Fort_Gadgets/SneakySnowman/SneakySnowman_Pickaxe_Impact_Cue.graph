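# Sound cue for the pickaxe impact on the Sneaky Snowman gadget.
# A Random node picks one of five equally weighted small impact waves from the Version2 folder and feeds it into a Modulator at the root.
# It plays under the gadget items sound class and borrows the grenade bounce attenuation asset for distance falloff.
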SoundCue /Game/Sounds/Fort_Gadgets/SneakySnowman/SneakySnowman_Pickaxe_Impact_Cue
root=Modulator_0
nodes [
 {id=Modulator_0, children [Random_0]}
 {id=Random_0, children [WavePlayer_0,WavePlayer_2,WavePlayer_3,WavePlayer_4,WavePlayer_5], properties {weights [1,1,1,1,1]}}
 {id=WavePlayer_0, properties {soundWaveAssetPtr=/Game/Sounds/Fort_Gadgets/SneakySnowman/Version2/SneakySnowman_Impact_Small_01}}
 {id=WavePlayer_2, properties {soundWaveAssetPtr=/Game/Sounds/Fort_Gadgets/SneakySnowman/Version2/SneakySnowman_Impact_Small_02}}
 {id=WavePlayer_3, properties {soundWaveAssetPtr=/Game/Sounds/Fort_Gadgets/SneakySnowman/Version2/SneakySnowman_Impact_Small_03}}
 {id=WavePlayer_4, properties {soundWaveAssetPtr=/Game/Sounds/Fort_Gadgets/SneakySnowman/Version2/SneakySnowman_Impact_Small_04}}
 {id=WavePlayer_5, properties {soundWaveAssetPtr=/Game/Sounds/Fort_Gadgets/SneakySnowman/Version2/SneakySnowman_Impact_Small_05}}
]
properties {
 attenuationSettings=/Game/Sounds/Attenuation/Fort_Grenade_Bounce_Att
 maxDistance=5500
 soundClassObject=&Fort_Gadgets_Items
 volumeMultiplier=1.0667
}
