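# Sound cue for the Sneaky Snowman gadget's possess sound.
# A mixer layers the main possess wave with a quieter, briefly delayed vocal "ouch" picked at random from three reused snowball launcher bounce lines.
# Each layer passes through its own modulator, and the cue borrows the grenade bounce attenuation within the gadget items sound class.
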SoundCue /Game/Sounds/Fort_Gadgets/SneakySnowman/SneakySnowman_Possess_Cue
root=Mixer_0
nodes [
 {id=Delay_1, children [Modulator_2], properties {delayMax=0.3, delayMin=0.2}}
 {id=Mixer_0, children [Modulator_0,Delay_1], properties {inputVolume [1,0.5]}}
 {id=Modulator_0, children [WavePlayer_1]}
 {id=Modulator_2, children [Random_3]}
 {id=Random_3, children [WavePlayer_7,WavePlayer_8,WavePlayer_9], properties {weights [1,1,1]}}
 {id=WavePlayer_1, properties {soundWaveAssetPtr=/Game/Sounds/Fort_Gadgets/SneakySnowman/Version2/SneakySnowman_Possess_01}}
 {id=WavePlayer_7, properties {soundWaveAssetPtr=/Game/Sounds/Fort_Weapons/Launchers/SnowBall_VO_Bounce_Ouch_01}}
 {id=WavePlayer_8, properties {soundWaveAssetPtr=/Game/Sounds/Fort_Weapons/Launchers/SnowBall_VO_Bounce_Ouch_03}}
 {id=WavePlayer_9, properties {soundWaveAssetPtr=/Game/Sounds/Fort_Weapons/Launchers/SnowBall_VO_Bounce_Ouch_02}}
]
properties {
 attenuationSettings=/Game/Sounds/Attenuation/Fort_Grenade_Bounce_Att
 maxDistance=5500
 soundClassObject=&Fort_Gadgets_Items
 volumeMultiplier=1.0667
}
