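# Sound cue for the Sneaky Snowman gadget's spawn, layering an impact sound with a vocal reaction.
# A mixer plays one of two random throw-impact waves at full level alongside a quieter, briefly delayed random snowball "ouch" voice line.
# Each layer passes through a modulator for variation.
# It uses grenade-bounce attenuation and the gadget items sound class.
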SoundCue /Game/Sounds/Fort_Gadgets/SneakySnowman/SneakySnowman_Spawn_Cue
root=Mixer_0
nodes [
 {id=Delay_0, children [Modulator_1], properties {delayMax=0.3, delayMin=0.2}}
 {id=Mixer_0, children [Modulator_2,Delay_0], properties {inputVolume [1,0.2]}}
 {id=Modulator_1, children [Random_1]}
 {id=Modulator_2, children [Random_3]}
 {id=Random_1, children [WavePlayer_0,WavePlayer_3,WavePlayer_4], properties {weights [1,1,1]}}
 {id=Random_3, children [WavePlayer_7,WavePlayer_8], properties {weights [1,1]}}
 {id=WavePlayer_0, properties {soundWaveAssetPtr=/Game/Sounds/Fort_Weapons/Launchers/SnowBall_VO_Bounce_Ouch_01}}
 {id=WavePlayer_3, properties {soundWaveAssetPtr=/Game/Sounds/Fort_Weapons/Launchers/SnowBall_VO_Bounce_Ouch_03}}
 {id=WavePlayer_4, properties {soundWaveAssetPtr=/Game/Sounds/Fort_Weapons/Launchers/SnowBall_VO_Bounce_Ouch_02}}
 {id=WavePlayer_7, properties {soundWaveAssetPtr=/Game/Sounds/Fort_Gadgets/SneakySnowman/Version2/SneakySnowman_Throw_Impact_01}}
 {id=WavePlayer_8, properties {soundWaveAssetPtr=/Game/Sounds/Fort_Gadgets/SneakySnowman/Version2/SneakySnowman_Throw_Impact_02}}
]
properties {
 attenuationSettings=/Game/Sounds/Attenuation/Fort_Grenade_Bounce_Att
 maxDistance=5500
 soundClassObject=&Fort_Gadgets_Items
 volumeMultiplier=1.0667
}
