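# Sound cue for the Sneaky Snowman gadget's unpossess sound.
# A single wave player feeds a modulator at default settings, which forms the output.
# It uses the grenade bounce attenuation in the gadget items sound class.
# A concurrency override limits it to one instance per owner, stopping the oldest when a new one starts.
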SoundCue /Game/Sounds/Fort_Gadgets/SneakySnowman/SneakySnowman_Unpossess_Cue
root=Modulator_0
nodes [
 {id=Modulator_0, children [WavePlayer_1]}
 {id=WavePlayer_1, properties {soundWaveAssetPtr=/Game/Sounds/Fort_Gadgets/SneakySnowman/Version2/SneakySnowman_Unpossess_01}}
]
properties {
 attenuationSettings=/Game/Sounds/Attenuation/Fort_Grenade_Bounce_Att
 bOverrideConcurrency=true
 maxDistance=5500
 soundClassObject=&Fort_Gadgets_Items
 volumeMultiplier=1.0667
 concurrencyOverrides {bLimitToOwner=true, maxCount=1, resolutionRule=StopOldest}
}
